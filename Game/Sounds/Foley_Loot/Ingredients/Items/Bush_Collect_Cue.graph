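# Sound cue for collecting a bush ingredient, classed as loot collect foley and using the prop-collect attenuation.
# A quality switch gives the first two levels a mix of real, half-volume surreal and bag-loot layers, while the last two play a single optimized wave.
# A concurrency override caps simultaneous instances at two, stopping the oldest.
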SoundCue /Game/Sounds/Foley_Loot/Ingredients/Items/Bush_Collect_Cue
root=QualityLevel_0
nodes [
 {id=Mixer_1, children [WavePlayer_6,WavePlayer_4,WavePlayer_8], properties {inputVolume [1,0.5,1]}}
 {id=QualityLevel_0, children [Mixer_1,Mixer_1,WavePlayer_0,WavePlayer_0]}
 {id=WavePlayer_0, properties {soundWaveAssetPtr=&/Bush_Collect_opt}}
 {id=WavePlayer_4, properties {soundWaveAssetPtr=&/Bush_Collect_Surreal}}
 {id=WavePlayer_6, properties {soundWaveAssetPtr=&/Bush_Collect_Real}}
 {id=WavePlayer_8, properties {soundWaveAssetPtr=/Game/Sounds/Foley_Loot/Ingredients/Bag/loot_setup_bagonly_01}}
]
properties {
 attenuationSettings=/Game/Sounds/Attenuation/Fort_Prop_Collect_Att
 bOverrideConcurrency=true
 maxDistance=2097152
 soundClassObject=&Fort_Loot_Collect
 volumeMultiplier=0.251
 concurrencyOverrides {maxCount=2, resolutionRule=StopOldest}
}
