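# Sound cue for collecting an electric trap item, played as a loot pickup sound in the Loot_Collect sound class.
# A mixer layers the electric wall trap "ready" wave at full level with a quieter generic special-loot-find effect.
# An enveloper on the generic effect holds its volume briefly, then dips it.
# The cue uses prop-collect attenuation and overrides concurrency to allow two instances, stopping the oldest.
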SoundCue /Game/Sounds/Foley_Loot/Ingredients/Items/Trap_Electric_Collect_Cue
root=Mixer_2
nodes [
 {
  id=Enveloper_0
  children [WavePlayer_10]
  properties {pitchCurve {editorCurveData {keys [{time=0}]}}, volumeCurve {editorCurveData {keys [<at 0 1>,<at 0.5 1>,<at 0.6716 0.3588>]}}}
 }
 {id=Mixer_2, children [WavePlayer_0,Enveloper_0], properties {inputVolume [1,0.7]}}
 {id=WavePlayer_0, properties {soundWaveAssetPtr=/Game/Sounds/Fort_Traps/Wall_Electric/WallTrap_Electric_Ready_001}}
 {id=WavePlayer_10, properties {soundWaveAssetPtr=/Game/Sounds/Foley_Loot/Ingredients/Generic/find_loot_special_effect}}
]
properties {
 attenuationSettings=/Game/Sounds/Attenuation/Fort_Prop_Collect_Att
 bOverrideConcurrency=true
 maxDistance=2097152
 priority=0.5
 soundClassObject=&Fort_Loot_Collect
 volumeMultiplier=0.501
 concurrencyOverrides {maxCount=2, resolutionRule=StopOldest}
 concurrencySet [/Game/Sounds/Concurrency_Rules/OneMaximum_Owner_StopLowestPriority]
}
