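# Sound cue for the Flame Grill trap equip foley, in the loot collect sound class.
# A mixer layers a quieter flame grill placement wave with a full-volume ballistic turret equip wave, reusing existing trap sounds.
# An enveloper fades the mix out in under a second and slightly raises its pitch.
# Overridden concurrency caps simultaneous instances, stopping the oldest first.
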SoundCue /Game/Sounds/Foley_Loot/Ingredients/Items/Trap_FlameGrill_Equip_Cue
root=Enveloper_2
nodes [
 {
  id=Enveloper_2
  children [Mixer_0]
  properties {pitchMax=1.1, pitchMin=1.05, pitchCurve {editorCurveData {keys [{time=0}]}}, volumeCurve {editorCurveData {keys [<at 0 1>,<at 0.65 0>]}}}
 }
 {id=Mixer_0, children [WavePlayer_3,WavePlayer_0], properties {inputVolume [0.4,1]}}
 {id=WavePlayer_0, properties {soundWaveAssetPtr=/Game/Sounds/Fort_Traps/MountedTurret/Turret_Ballistic_Equip_01}}
 {id=WavePlayer_3, properties {soundWaveAssetPtr=/Game/Sounds/Fort_Traps/Flame_Trap/FlameGrill_Trap_Place_01}}
]
properties {
 bOverrideConcurrency=true
 maxDistance=2097152
 priority=0.5
 soundClassObject=&Fort_Loot_Collect
 volumeMultiplier=0.398
 concurrencyOverrides {maxCount=2, resolutionRule=StopOldest}
 concurrencySet [/Game/Sounds/Concurrency_Rules/OneMaximum_Owner_StopLowestPriority]
}
