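# Sound cue for collecting a gas trap ingredient, layering a gas trap activation sound with a generic special loot-find effect.
# A mixer plays both waves at equal volume, each shaped by an enveloper: the gas hiss fades in from near silence while the loot effect holds, then dips.
# It uses the prop-collect attenuation and loot-collect sound class at reduced volume, with concurrency limits to keep repeated pickups from stacking.
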SoundCue /Game/Sounds/Foley_Loot/Ingredients/Items/Trap_Gas_Collect_Cue
root=Mixer_2
nodes [
 {
  id=Enveloper_0
  children [WavePlayer_10]
  properties {pitchCurve {editorCurveData {keys [{time=0}]}}, volumeCurve {editorCurveData {keys [<at 0 1>,<at 0.5 1>,<at 0.6716 0.3588>]}}}
 }
 {
  id=Enveloper_1
  children [WavePlayer_1]
  properties {pitchCurve {editorCurveData {keys [{time=0}]}}, volumeCurve {editorCurveData {keys [<at 0.0001 0.0123>,<at 0.3 1>]}}}
 }
 {id=Mixer_2, children [Enveloper_1,Enveloper_0], properties {inputVolume [1,1]}}
 {id=WavePlayer_1, properties {soundWaveAssetPtr=/Game/Sounds/Fort_Traps/Ceiling_Gas_Trap/CeilingTrap_Gas_Activate_001}}
 {id=WavePlayer_10, properties {soundWaveAssetPtr=/Game/Sounds/Foley_Loot/Ingredients/Generic/find_loot_special_effect}}
]
properties {
 attenuationSettings=/Game/Sounds/Attenuation/Fort_Prop_Collect_Att
 bOverrideConcurrency=true
 maxDistance=2097152
 priority=0.5
 soundClassObject=&Fort_Loot_Collect
 volumeMultiplier=0.501
 concurrencyOverrides {maxCount=2, resolutionRule=StopOldest}
 concurrencySet [/Game/Sounds/Concurrency_Rules/OneMaximum_Owner_StopLowestPriority]
}
