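# Sound cue for collecting a spike trap item, played on the loot-collect sound class with prop-collect attenuation.
# A mixer layers a floor-spike trap "ready" wave with a generic loot-find effect wave.
# An enveloper holds the loot effect at full volume briefly, then lowers it to a reduced level.
# Assigned concurrency rules limit overlapping plays, with the overall cue volume reduced to about half.
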SoundCue /Game/Sounds/Foley_Loot/Ingredients/Items/Trap_Spike_Collect_Cue
root=Mixer_2
nodes [
 {
  id=Enveloper_0
  children [WavePlayer_10]
  properties {pitchCurve {editorCurveData {keys [{time=0}]}}, volumeCurve {editorCurveData {keys [<at 0 1>,<at 0.5 1>,<at 0.6716 0.3588>]}}}
 }
 {id=Mixer_2, children [WavePlayer_2,Enveloper_0], properties {inputVolume [1,1]}}
 {id=WavePlayer_10, properties {soundWaveAssetPtr=/Game/Sounds/Foley_Loot/Ingredients/Generic/find_loot_special_effect}}
 {id=WavePlayer_2, properties {soundWaveAssetPtr=/Game/Sounds/Fort_Traps/Spike_Trap/FloorTrap_Spikes_Ready_002}}
]
properties {
 attenuationSettings=/Game/Sounds/Attenuation/Fort_Prop_Collect_Att
 maxDistance=2097152
 priority=0.6
 soundClassObject=&Fort_Loot_Collect
 volumeMultiplier=0.501
 concurrencyOverrides {maxCount=2, resolutionRule=StopOldest}
 concurrencySet [/Game/Sounds/Concurrency_Rules/OneMaximum_Owner_StopLowestPriority]
}
